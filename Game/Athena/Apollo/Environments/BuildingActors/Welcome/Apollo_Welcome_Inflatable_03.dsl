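; Apollo_Welcome_Inflatable_03 is an inflatable building actor for the Apollo Welcome environment that varies its material animation per instance.
; On BeginPlay it calls RandomizeWPO, which creates a dynamic material instance on slot 0 of its static mesh component.
; It sets that material's TimeOffset parameter to a random value, apparently so world position offset motion differs between placed copies.
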(event EventBeginPlay
  (CallFunction|RandomizeWPO))

(fn randomizeWPO
  (bind _returnvalue (Rendering|Material|CreateDynamicMaterialInstance (Variables|Debug|GetStaticMeshComponent) 0 0 "inflatableMID" :!class /Script/Engine.PrimitiveComponent))
  (Rendering|Material|SetScalarParameterValue _returnvalue "TimeOffset" (* (Math|Random|RandomFloat) 11.0) :!class /Script/Engine.MaterialInstanceDynamic))

(fn ConstructionScript)
